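# Niagara system for the swing anim trails of the RankedOlympus pickaxe, drawn as three short-lived ribbon emitters: BlueTrail, SoftDarkTrail and TipWhiteTrail.
# Each emitter uses the shared FN anim-trail spawn, initialize and update modules, sampling the system's skeletal mesh.
# Spawn colors are selected by time of day, with cyan values only on SoftDarkTrail.
# Update color curves fade alpha, scaled by system alpha and time-of-day brightness parameters.
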NiagaraSystem /BRCosmetics/Weapons/FORT_Melee/Pickaxe_RankedOlympus/FX/NS_Pickaxe_RankedOlympus_AnimTrails
dependencies {
 usedDataInterfaces [&NiagaraDataInterfaceColorCurve, &NiagaraDataInterfaceSkeletalMesh, &NiagaraDataInterfaceCurve]
 usedDynamicInputs [&FloatFromCurve, &FNTimeOfDayScalar]
 usedModules [
  &self:SystemSpawnScript.SystemScriptSource.SystemScriptGraph.NiagaraNodeAssignment_0.SetVariables_#1
  &SystemState
  &self:SystemSpawnScript.SystemScriptSource.SystemScriptGraph.NiagaraNodeAssignment_1.SetVariables_#2
  &self:SystemSpawnScript.SystemScriptSource.SystemScriptGraph.NiagaraNodeAssignment_2.SetVariables_#3
 ]
}
emitters {
 BlueTrail {
  emitterUpdateScript [
   &EmitterState()
   &FN_SpawnAnimTrail()
  ]
  particleSpawnScript [
   &FN_InitializeAnimTrailBoneIndecies()
   &FN_InitializeAnimTrail(
    Color=&FN_Select_Color_from_Time_Of_Day(Morning=<rgba 1 1 1 1>, Day=<rgba 1 1 1 1>, Evening=<rgba 1 1 1 1>, Night=<rgba 1 1 1 1>)
    Lifetime=0.1
    SkeletalMesh=$System.SkeletalMesh
   )
  ]
  particleUpdateScript [
   &FN_UpdateAnimTrail(SkeletalMesh=$System.SkeletalMesh) {module=FN_UpdateAnimTrail001}
   &Color(
    Color=&ColorFromCurve(ColorCurve=DataInterface() {propertyValues {
       AlphaCurve {keys [<at 0.0935 0.4>,<at 0.545 0 RCIM_Cubic>]}
       BlueCurve {keys [<at 0 0.3>]}
       GreenCurve {keys [<at 0 0.285>]}
       RedCurve {keys [<at 0 0.2868>]}
      }})
    Scale Alpha=$System.ScaleAlpha
    Scale Color=&VectorFromFloat(Value=$System.TimeOfDayBrightnessMultiplier)
   )
   &SolveForcesAndVelocity(Clamp Velocity=false, Limit Acceleration=false)
  ]
  rendererClasses [&NiagaraRibbonRendererProperties]
  renderers [{rendererClass=&NiagaraRibbonRendererProperties}]
 }
 SoftDarkTrail {
  emitterUpdateScript [
   &EmitterState()
   &FN_SpawnAnimTrail()
  ]
  particleSpawnScript [
   &FN_InitializeAnimTrailBoneIndecies()
   &FN_InitializeAnimTrail(
    Color=&FN_Select_Color_from_Time_Of_Day(Morning=<rgba 0 2.8891 4 1>, Day=<rgba 0 3.8853 5 1>, Evening=<rgba 0 3.5105 4 1>, Night=<rgba 0 3.8978 5 1>)
    Lifetime=0.1
    SkeletalMesh=$System.SkeletalMesh
   )
  ]
  particleUpdateScript [
   &FN_UpdateAnimTrail(SkeletalMesh=$System.SkeletalMesh) {module=FN_UpdateAnimTrail001}
   &Color(
    Color=&ColorFromCurve(ColorCurve=DataInterface() {propertyValues {
       AlphaCurve {keys [<at 0 0.2 RCIM_Cubic>,<at 1 0 RCIM_Cubic>]}
       BlueCurve {keys [<at 0 0.03>]}
       GreenCurve {keys [<at 0 0.03>]}
       RedCurve {keys [<at 0 0.03>]}
      }})
    Scale Alpha=$System.ScaleAlpha
    Scale Color=&VectorFromFloat(Value=$System.TimeOfDayBrightnessMultiplier)
   )
   &SolveForcesAndVelocity(Clamp Velocity=false, Limit Acceleration=false)
  ]
  rendererClasses [&NiagaraRibbonRendererProperties]
  renderers [{rendererClass=&NiagaraRibbonRendererProperties}]
 }
 TipWhiteTrail {
  emitterUpdateScript [
   &EmitterState()
   &FN_SpawnAnimTrail()
  ]
  particleSpawnScript [
   &FN_InitializeAnimTrailBoneIndecies()
   &FN_InitializeAnimTrail(
    Color=&FN_Select_Color_from_Time_Of_Day(Morning=<rgba 1 1 1 1>, Day=<rgba 1 1 1 1>, Evening=<rgba 1 1 1 1>, Night=<rgba 1 1 1 1>)
    Lifetime=0.1
    SkeletalMesh=$System.SkeletalMesh
   )
  ]
  particleUpdateScript [
   &FN_UpdateAnimTrail(SkeletalMesh=$System.SkeletalMesh) {module=FN_UpdateAnimTrail001}
   &Color(
    Color=&ColorFromCurve(ColorCurve=DataInterface() {propertyValues {
       AlphaCurve {keys [<at 0 0.4>,<at 0.5 0 RCIM_Cubic>]}
       BlueCurve {keys [<at 0 0.45>]}
       GreenCurve {keys [<at 0 0.45>]}
       RedCurve {keys [<at 0 0.45>]}
      }})
    Scale Alpha=$System.ScaleAlpha
    Scale Color=&VectorFromFloat(Value=$System.TimeOfDayBrightnessMultiplier)
   )
   &SolveForcesAndVelocity(Clamp Velocity=false, Limit Acceleration=false)
  ]
  rendererClasses [&NiagaraRibbonRendererProperties]
  renderers [{rendererClass=&NiagaraRibbonRendererProperties}]
 }
}
summary {
 systemName=NS_Pickaxe_RankedOlympus_AnimTrails
 emitters [
  {emitterName=SoftDarkTrail, rendererClasses [&NiagaraRibbonRendererProperties]}
  {emitterName=BlueTrail, rendererClasses [&NiagaraRibbonRendererProperties]}
  {emitterName=TipWhiteTrail, rendererClasses [&NiagaraRibbonRendererProperties]}
 ]
 userVariables [User.RibbonIDOffset {type=NiagaraInt32}]
}
